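# Unlit translucent material for Niagara mesh particles, evidently the core of a stage-FX sparks and smoke cylinder.
# An inverted smoke texture pans at a dynamic-parameter speed with per-particle random offset and optional flip, alongside a UV-gradient fire mask.
# A curve-atlas ramp tinted by particle color builds the emissive, while particle alpha erodes opacity.
# Three layered, time-scrolled sine waves form a height-weighted wobble along vertex normals.
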Material /POIBanner_Dance/Assets/FX/Materials/M_Sparks_Smoke_Cylinder_Core
graph {
 properties {
  bUsedWithNiagaraMeshParticles=true
  blendMode=BLEND_Translucent
  pixelDepthOffsetMode=PDOM_Legacy
  shadingModel=MSM_Unlit
  usedShadingModels=Unlit
 }
 nodes [
  n1 = NamedRerouteUsage()
  CurveAtlasRowParameter(CurveTime=@n1) {
   atlas=/POIBanner_Dance/Assets/FX/Fire/Textures/CurveAtlas_StageFX_Fire
   curve=/POIBanner_Dance/Assets/FX/Fire/Textures/Curve_StageFX_Fire_Natural
   group=Fire Color Ramp Control
   parameterName=ColorCurve
   sortPriority=3
  }
  n2 = ScalarParameter() {defaultValue=2, group=Fire Color Ramp Control, parameterName=Final Intensity Boost, sortPriority=2}
  MaterialFunctionCall(Emissive=Multiply(A=ParticleColor().RGB, B=Multiply(A=@n2, B=@n1))) {materialFunction=&MF_FX_EmissiveAdjustments}
  n3 = TextureCoordinate()
  n4 = Time()
  n5 = Multiply(A=Add(A=ComponentMask(None=@n3) {g=true}, B=Multiply(A=@n4) {constB=2})) {constB=5}
  n6 = TextureCoordinate()
  n7 = Time()
  n8 = Multiply(A=Add(A=ComponentMask(None=@n6) {g=true}, B=Multiply(A=@n7) {constB=1.75})) {constB=3}
  n9 = TextureCoordinate()
  n10 = Time()
  n11 = Multiply(A=Add(A=ComponentMask(None=@n9) {g=true}, B=Multiply(A=@n10) {constB=1.25})) {constB=0.75}
  n12 = LinearInterpolate(A=ConstantBiasScale(None=Sine(None=@n5)), B=ConstantBiasScale(None=Sine(None=@n8)), Alpha=ConstantBiasScale(None=Sine(None=@n11)))
  n13 = VertexNormalWS()
  n14 = ComponentMask(None=OneMinus(None=TextureCoordinate())) {g=true}
  Multiply(
   A=Multiply(A=Multiply(A=@n12, B=@n13), B=Add(A=@n14) {constB=0.15}) {constB=5}
   B=ScalarParameter() {defaultValue=8, parameterName=Wobble Strength}
  )
  n15 = TextureCoordinate()
  n16 = Divide(B=OneMinus(None=ComponentMask(None=@n15) {g=true})) {constA=0.23}
  n17 = ParticleColor()
  n18 = NamedRerouteUsage()
  n19 = Subtract(A=Saturate(None=Add(A=@n17.A, B=@n18)), B=OneMinus(None=@n17.A))
  n20 = ScalarParameter() {defaultValue=0.2, parameterName=Opacity}
  Multiply(A=Multiply(A=Add(B=@n16) {constA=-0.28}, B=NamedRerouteUsage()), B=Divide(A=@n19, B=@n20) {constB=1}) {constA=1}
  n21 = OneMinus(None=TextureCoordinate())
  n22 = Multiply(A=ComponentMask(None=TextureCoordinate()) {g=true}, B=SmoothStep(Value=ComponentMask(None=@n21) {g=true}) {constMax=0.1, constMin=0.1})
  n23 = NamedRerouteUsage()
  NamedRerouteDeclaration(None=Saturate(None=Multiply(A=@n22, B=@n23))) {name=FireMask, nodeColor=<rgba 0 0 0 1>}
  n24 = ParticleRandom()
  n25 = TextureCoordinate()
  n26 = ComponentMask(None=@n25) {r=true}
  n27 = OneMinus(None=@n26)
  n28 = ComponentMask(None=@n25) {g=true}
  n29 = StaticSwitchParameter(True=AppendVector(A=If(A=@n24, "A > B"=@n26, "A == B"=@n27, "A < B"=@n27) {constB=0.5}, B=OneMinus(None=@n28)), False=TextureCoordinate()) {defaultValue=true, group=Fire Texture, parameterName=Flip Fire Texture}
  n30 = AppendVector(
   A=ScalarParameter() {defaultValue=1, desc=1, group=Fire Texture, parameterName=Fire Tile X, sortPriority=0}
   B=ScalarParameter() {defaultValue=1, desc=0.75, group=Fire Texture, parameterName=Fire Tile Y, sortPriority=1}
  )
  n31 = Time()
  n32 = ScalarParameter() {defaultValue=1, parameterName=Fire Tex Speed}
  n33 = Panner(
   Coordinate=Add(A=Multiply(A=@n29, B=@n30), B=ScalarParameter() {defaultValue=0.6, desc=0.1, group=Fire Texture, parameterName=Fire Tex Pos, sortPriority=2})
   Time=Add(A=Multiply(A=@n31, B=@n32), B=ParticleRandom())
   Speed=AppendVector(A=Constant(), B=DynamicParameter() {defaultValue=<rgba 1 -2.5 1 1>, paramNames [Temperature, Fire Speed, Param3, Param4]}.Fire Speed)
  ) {speedY=0.75}
  NamedRerouteDeclaration(None=OneMinus(None=TextureSample(UVs=@n33) {texture=&T_Inky_Smoke_Tile}.G)) {name=FireTex, nodeColor=<rgba 0 0 0 1>}
  Saturate(None=Multiply(A=@n19, B=@n20))
 ]
}
