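# Translucent unlit Niagara sprite material for stage-show smoke can effects, using an 8x8 falling-mist flipbook tinted by particle color.
# Opacity combines particle alpha with an erode-powered flipbook mask, depth-faded off mobile, and a per-frame UV gradient.
# Color blends smoke, optionally lit by a noise-distorted spherical glow around the actor, with a top/bottom gradient.
# A flattened sphere normal dotted with the atmospheric light fakes shadowing, skipped on low and medium quality.
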Material /POIBanner_Dance/Assets/FX/Materials/M_StageShow_SmokeCan
graph {
 properties {
  bAutomaticallySetUsageInEditor=false
  bUsedWithNiagaraSprites=true
  blendMode=BLEND_Translucent
  pixelDepthOffsetMode=PDOM_Legacy
  shadingModel=MSM_Unlit
  twoSided=true
  usedShadingModels=Unlit
 }
 nodes [
  n1 = ParticleColor()
  n2 = TextureObjectParameter() {parameterName=TextureObject, samplerType=SAMPLERTYPE_LinearColor, texture=/Game/Effects/Fort_Effects/Textures/Smoke/T_FallingMist_8x8_01}
  n3 = ParticleSubUV!(UVs, Tex=@n2, Apply View MipBias) {samplerType=SAMPLERTYPE_LinearColor}
  n4 = Multiply(A=@n1.A, B=Power(Base=@n3.B, Exp=DynamicParameter() {paramNames [Erode Power, --, --, --]}.Erode Power))
  n5 = ShadingPathSwitch(Default=DepthFade(Opacity=@n4), Mobile=@n4)
  n6 = NamedRerouteUsage()
  n7 = Power(Base=Modulo!(A=ComponentMask(None=@n6) {g=true}, B=Constant() {r=1}), Exp=ScalarParameter() {defaultValue=2, parameterName=Gradient Power})
  LinearInterpolate(A=@n5, B=Multiply(A=OneMinus(None=@n7), B=@n5), Alpha=ScalarParameter() {defaultValue=1, parameterName=Apply Alpha Gradient})
  n8 = ScalarParameter() {defaultValue=0.6, parameterName=Shadow Color Mult}
  n9 = WorldPosition()
  n10 = VectorParameter() {parameterName=Flipbook Dimensions, channelNames {a="", b="", g=V Frames, r=U Frames}, defaultValue=<rgba 8 8 0 0>}
  n11 = NamedRerouteDeclaration(None=Multiply(A=TextureCoordinate(), B=ComponentMask(None=@n10) {g=true, r=true})) {name=Tiled UVs, nodeColor=<rgba 1 0 0.3281 1>}
  n12 = Multiply(
   A=OneMinus(None=TextureSample(UVs=@n11) {texture=/Game/Effects/FogVolume/noise_flamy_Tile}.G)
   B=ScalarParameter() {defaultValue=250, parameterName=Glow Distort}
  )
  n13 = Multiply(
   A=SphereMask(
    A=Add(A=@n9.XYZ, B=@n12)
    B=ActorPositionWS()
    Radius=ScalarParameter() {defaultValue=750, parameterName=Glow Radius}
    Hardness=ScalarParameter() {defaultValue=0.5, parameterName=Glow Hardness}
   )
   B=VectorParameter() {parameterName=Flame Glow, defaultValue=<rgba 2 0.1667 0 1>}.RGB
  )
  n14 = Multiply(A=@n3.B, B=@n1.RGB)
  n15 = VectorParameter() {parameterName=Gradient Top, defaultValue=<rgba 1 1 1 1>}
  n16 = VectorParameter() {parameterName=Gradient Bottom, defaultValue=<rgba 1 1 1 1>}
  n17 = Multiply(A=@n1.RGB, B=@n16.RGB)
  n18 = LinearInterpolate(
   A=StaticSwitchParameter(True=Multiply(A=@n13, B=@n14), False=@n14) {parameterName=Use Sphereical Glow}
   B=LinearInterpolate(A=Multiply(A=@n15.RGB, B=@n1.RGB), B=LinearInterpolate(A=@n17, B=@n16.RGB), Alpha=@n7)
   Alpha=ScalarParameter() {defaultValue=1, parameterName=Apply Color Gradient}
  )
  n19 = MaterialFunctionCall(
   Normal=TextureSample(UVs=NamedRerouteUsage()) {samplerType=SAMPLERTYPE_Normal, texture=&T_Sphere_N}.RGB
   Flatness=ScalarParameter() {defaultValue=0.9, parameterName=Flatten Sphere Normal}
  ) {materialFunction=&FlattenNormal}
  n20 = Saturate(None=DotProduct(A=Transform(None=@n19.Result), B=AtmosphericLightVector()))
  n21 = ScalarParameter() {defaultValue=2, parameterName=Shadow Power}
  QualitySwitch(Default=LinearInterpolate(A=Multiply(A=@n8, B=@n18), B=@n18, Alpha=Power(Base=@n20, Exp=@n21)), Low=@n18, Medium=@n18)
 ]
}
