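# One-shot stereo MetaSound source for the Banner Dance firework explosion, with distant-fireworks attenuation and a concurrency override.
# On play, a repeat trigger drives a two-way sequence, firing two layers that each pick a random wave from an array and randomize its pitch.
# The second, lower-pitched layer fades out with attenuation distance in the mixer, leaving the first layer at range.
# It finishes once both players end; the Mobile page duplicates the Default graph.
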MetaSoundSource /POIBanner_Dance/Gameplay/Actors/Audio/MSS_Firework_Explode_BannerDance
inputs [
 UE.Attenuation.Distance {type=Float}
 UE.Source.OnPlay {type=Trigger}
]
interfaces [UE.Attenuation:1.0, UE.OutputFormat.Stereo:1.0, UE.Source.OneShot:1.0, UE.Source:1.1]
outputs [
 "UE.OutputFormat.Stereo.Audio:0" {type=Audio}
 "UE.OutputFormat.Stereo.Audio:1" {type=Audio}
 UE.Source.OneShot.OnFinished {type=Trigger}
]
pages [
 Default {
  is_build_graph=true
  input_defaults {UE.Attenuation.Distance=0, UE.Source.OnPlay=null}
  nodes [
   TriggerSequence2_0 = "UE.Trigger Sequence (2)"(In=UE.TriggerRepeat(Start=^UE.Source.OnPlay) {also {defaults {Num Repeats=2}}}.RepeatOut)
   ArrayRandomGet_0 = Array.Random Get:WaveAsset:Array(Next=@TriggerSequence2_0.Out 1, In Array=VariableAccessor.WaveAsset:Array(Variable=InitVariable.WaveAsset:Array().Variable).Value) {also {defaults {Enable Shared State="<MetaSoundBuilderResult.SUCCEEDED: 0>", No Repeats=-1, Weights []}}}
   n1 = UE.RandomFloat(Next=@ArrayRandomGet_0.On Next) {also {defaults {Max=-1, Min=-1.5}}}
   WavePlayer_0 = UE.Wave Player:Stereo(Wave Asset=@ArrayRandomGet_0.Value, Play=@n1.On Next, Pitch Shift=@n1.Value) {also {version [1,1]}}
   ArrayRandomGet_1 = Array.Random Get:WaveAsset:Array(Next=@TriggerSequence2_0.Out 0, In Array=VariableAccessor.WaveAsset:Array(Variable=InitVariable.WaveAsset:Array().Variable).Value) {also {defaults {Enable Shared State="<MetaSoundBuilderResult.SUCCEEDED: 0>", No Repeats=-1, Weights []}}}
   n2 = UE.RandomFloat(Next=@ArrayRandomGet_1.On Next) {also {defaults {Min=-1}}}
   WavePlayer_1 = UE.Wave Player:Stereo(Wave Asset=@ArrayRandomGet_1.Value, Play=@n2.On Next, Pitch Shift=@n2.Value) {also {version [1,1]}}
   AudioMixerAudioMixerStereo2_0 = "AudioMixer.Audio Mixer (Stereo, 2)"(
    Gain 1=MapRange.MapRange:Float(In=^UE.Attenuation.Distance) {also {defaults {In Range A=1500, In Range B=3500, Out Range A=1, Out Range B=0}}}.Out Value
    In 1 L=@WavePlayer_0.Out Left
    In 1 R=@WavePlayer_0.Out Right
    In 0 L=@WavePlayer_1.Out Left
    In 0 R=@WavePlayer_1.Out Right
   )
   TriggerAccumulatorTriggerAccumulate2_0 = "TriggerAccumulator.Trigger Accumulate (2)"(In 1=@WavePlayer_0.On Finished, In 0=@WavePlayer_1.On Finished)
  ]
  links [
   {dst="@UE.OutputFormat.Stereo.Audio:0", src=AudioMixerAudioMixerStereo2_0.Out L}
   {dst="@UE.OutputFormat.Stereo.Audio:1", src=AudioMixerAudioMixerStereo2_0.Out R}
   {dst="@UE.Source.OneShot.OnFinished", src=TriggerAccumulatorTriggerAccumulate2_0.Out}
  ]
 }
 Mobile {
  nodes [
   TriggerSequence2_0 = "UE.Trigger Sequence (2)"(In=UE.TriggerRepeat(Start=^UE.Source.OnPlay) {also {defaults {Num Repeats=2}}}.RepeatOut)
   ArrayRandomGet_0 = Array.Random Get:WaveAsset:Array(Next=@TriggerSequence2_0.Out 1, In Array=VariableAccessor.WaveAsset:Array(Variable=InitVariable.WaveAsset:Array().Variable).Value) {also {defaults {Enable Shared State="<MetaSoundBuilderResult.SUCCEEDED: 0>", No Repeats=-1, Weights []}}}
   n1 = UE.RandomFloat(Next=@ArrayRandomGet_0.On Next) {also {defaults {Max=-1, Min=-1.5}}}
   WavePlayer_0 = UE.Wave Player:Stereo(Wave Asset=@ArrayRandomGet_0.Value, Play=@n1.On Next, Pitch Shift=@n1.Value) {also {version [1,1]}}
   ArrayRandomGet_1 = Array.Random Get:WaveAsset:Array(Next=@TriggerSequence2_0.Out 0, In Array=VariableAccessor.WaveAsset:Array(Variable=InitVariable.WaveAsset:Array().Variable).Value) {also {defaults {Enable Shared State="<MetaSoundBuilderResult.SUCCEEDED: 0>", No Repeats=-1, Weights []}}}
   n2 = UE.RandomFloat(Next=@ArrayRandomGet_1.On Next) {also {defaults {Min=-1}}}
   WavePlayer_1 = UE.Wave Player:Stereo(Wave Asset=@ArrayRandomGet_1.Value, Play=@n2.On Next, Pitch Shift=@n2.Value) {also {version [1,1]}}
   AudioMixerAudioMixerStereo2_0 = "AudioMixer.Audio Mixer (Stereo, 2)"(
    Gain 1=MapRange.MapRange:Float(In=^UE.Attenuation.Distance) {also {defaults {In Range A=1500, In Range B=3500, Out Range A=1, Out Range B=0}}}.Out Value
    In 1 L=@WavePlayer_0.Out Left
    In 1 R=@WavePlayer_0.Out Right
    In 0 L=@WavePlayer_1.Out Left
    In 0 R=@WavePlayer_1.Out Right
   )
   TriggerAccumulatorTriggerAccumulate2_0 = "TriggerAccumulator.Trigger Accumulate (2)"(In 1=@WavePlayer_0.On Finished, In 0=@WavePlayer_1.On Finished)
  ]
  links [
   {dst="@UE.OutputFormat.Stereo.Audio:0", src=AudioMixerAudioMixerStereo2_0.Out L}
   {dst="@UE.OutputFormat.Stereo.Audio:1", src=AudioMixerAudioMixerStereo2_0.Out R}
   {dst="@UE.Source.OneShot.OnFinished", src=TriggerAccumulatorTriggerAccumulate2_0.Out}
  ]
 }
]
properties {
 assetImportData=&self:AssetImportData
 attenuationSettings=/Game/Sounds/Attenuation/Fort_Weapon_Fireworks_Distant_Att
 bDebug=true
 bOverrideConcurrency=true
 maxDistance=20000
 numChannels=2
 outputFormat=Stereo
 sampleRate=48000
 soundClassObject=/Game/Sounds/FortSoundClassesAndModes/Fort_Explosives_Medium
 concurrencyOverrides {maxCount=5}
}
